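# Ambient sound cue for the evil dynamic wall environment actor, layering a continuous material loop with intermittent accents.
# A Mixer combines a slightly boosted looping bed with a Looping branch that plays one of four equally weighted random accents after a random multi-second delay.
# The bed passes through a Quality Level node whose inputs all share one wave player.
# The cue uses the ambient sound class, reverb submix send and a low concurrency cap.
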SoundCue /Game/Sounds/Asteria/Environment_Actors/EvilDynamicWall/Spot_Evil_Dynamic_Wall_Accent_Loop_Cue
root=Mixer_0
nodes [
 {id=Delay_0, children [Random_1], properties {delayMax=12, delayMin=4}}
 {id=Looping_0, children [Delay_0]}
 {id=Mixer_0, children [Looping_0, QualityLevel_0], properties {inputVolume [1,1.2]}}
 {id=QualityLevel_0, children [WavePlayer_0,WavePlayer_0,WavePlayer_0,WavePlayer_0]}
 {id=Random_1, children [WavePlayer_6,WavePlayer_7,WavePlayer_8,WavePlayer_9], properties {weights [1,1,1,1]}}
 {
  id=WavePlayer_0
  properties {bLooping=true, soundWaveAssetPtr=&/Spot_Evil_Dynamic_Wall_Material_Loop}
 }
 {
  id=WavePlayer_6
  properties {soundWaveAssetPtr=&/Spot_Evil_Dynamic_Wall_Material_Accent_04}
 }
 {
  id=WavePlayer_7
  properties {soundWaveAssetPtr=&/Spot_Evil_Dynamic_Wall_Material_Accent_01}
 }
 {
  id=WavePlayer_8
  properties {soundWaveAssetPtr=&/Spot_Evil_Dynamic_Wall_Material_Accent_02}
 }
 {
  id=WavePlayer_9
  properties {soundWaveAssetPtr=&/Spot_Evil_Dynamic_Wall_Material_Accent_03}
 }
]
properties {
 bOverrideConcurrency=true
 maxDistance=2097152
 soundClassObject=&Fort_Ambient_Athena
 volumeMultiplier=0.501
 attenuationOverrides {
  bEnableFocusInterpolation=true
  falloffDistance=600
  nonFocusVolumeAttenuation=0.3
  occlusionVolumeAttenuation=0.3
  stereoSpread=0
  attenuationShapeExtents {x=200}
 }
 concurrencyOverrides {maxCount=2}
 soundSubmixSends [{minSendLevel=0.4, soundSubmix=&Fortnite_Reverb_Default}]
}
